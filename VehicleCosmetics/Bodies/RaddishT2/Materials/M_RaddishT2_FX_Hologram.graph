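# Translucent, unlit, two-sided hologram material for RaddishT2 vehicle skeletal meshes, tinted by a CustomColor parameter.
# Vertex red blends two attribute sets: moving sine bands along U, sharpened by a power, and a constant half-opacity layer with squared color.
# Both share a small looping wobble in World Position Offset, built from sines and cosine of time offset by vertex red, then position-transformed from world space.
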Material /VehicleCosmetics/Bodies/RaddishT2/Materials/M_RaddishT2_FX_Hologram
graph {
 properties {
  bAutomaticallySetUsageInEditor=false
  bUseMaterialAttributes=true
  bUsedWithSkeletalMesh=true
  blendMode=BLEND_Translucent
  previewMesh=/VehicleCosmetics/Bodies/Raddish/RaddishT2/Meshes/Attachments/SK_RaddishT2_Attachment
  shadingModel=MSM_Unlit
  twoSided=true
  usedShadingModels=Unlit
 }
 nodes [
  n1 = VectorParameter() {parameterName=CustomColor, defaultValue=<rgba 0 0.6441 1 2>}
  n2 = Multiply(A=Time()) {constB=0.5}
  n3 = ComponentMask(None=TextureCoordinate()) {r=true}
  n4 = Power(Base=Sine(None=Add(A=@n2, B=@n3))) {constExponent=12}
  n5 = Add(A=Time(), B=VertexColor().R)
  n6 = AppendVector(A=Sine(None=@n5) {period=2}, B=Cosine(None=@n5) {period=2})
  n7 = Sine(None=@n5) {period=1.333}
  n8 = Multiply(
   A=TransformPosition(None=AppendVector(A=@n6, B=@n7)) {transformSourceType=TRANSFORMPOSSOURCE_World}
   B=Constant3Vector() {constant=<rgba 0.1 0.1 0.1 0>}
  )
  BlendMaterialAttributes(
   A=SetMaterialAttributes!(MaterialAttributes, Emissive Color=@n1.RGB, Opacity=Multiply(A=@n4) {constB=0.5}, World Position Offset=@n8) {attributeSetTypes [#1, #2, #3]}
   B=SetMaterialAttributes(Emissive Color=Multiply(A=@n1.RGB, B=@n1.RGB), Opacity=Constant() {r=0.5}, World Position Offset=@n8) {attributeSetTypes [#1, #2, #3]}
   Alpha=VertexColor().R
  )
  Constant3Vector()
 ]
}
